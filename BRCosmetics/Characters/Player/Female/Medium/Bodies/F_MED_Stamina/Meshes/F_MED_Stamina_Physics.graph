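# Physics asset for the F_MED_Stamina female player body mesh, set up for secondary motion on its scarf.
# The knot hangs off the neck with tight angular limits and branches into a two-capsule back chain and a three-box front chain, with wider swing toward the tips.
# Capsules on the neck, upper spine and left upper arm, joined by angularly locked constraints, form the body volume around the scarf.
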PhysicsAsset /BRCosmetics/Characters/Player/Female/Medium/Bodies/F_MED_Stamina/Meshes/F_MED_Stamina_Physics
bodies {
 dyn_scarf_bk_01 [{
   shapeType=Capsule
   length=1
   radius=2.5
   shapeName=dyn_scarf_bk_01_capsule
   center=<-0.166 0.0293 -0.2163>
   rotation=<rot 90 -169.3803 -159.3805>
  }]
 dyn_scarf_bk_02 [{shapeType=Capsule, length=0, radius=2, shapeName=dyn_scarf_bk_02_capsule}]
 dyn_scarf_fr_01 [{
   shapeType=Box
   extentX=5.5
   extentY=2
   extentZ=3.5
   shapeName=dyn_scarf_fr_01_box
   center=<0 -1.723 -0.177>
   rotation=<rot 0.0003 20.0001 90>
  }]
 dyn_scarf_fr_02 [{shapeType=Box, extentX=7.5, extentY=2.5, extentZ=7.5, shapeName=dyn_scarf_fr_02_box, rotation=<rot 12.393 18.6498 120.2749>}]
 dyn_scarf_fr_03 [{
   shapeType=Box
   extentX=3.5
   extentY=2
   extentZ=5
   shapeName=dyn_scarf_fr_03_box
   center=<-1.0361 1.172 0.2957>
   rotation=<rot 1.2385 22.3059 -29.756>
  }]
 dyn_scarf_knot [{shapeType=Capsule, length=1.25, radius=1.5, shapeName=dyn_scarf_knot_capsule, center=<0.4015 -0.5115 0>, rotation=<rot 10.0001 0 0>}]
 neck_01 [{shapeType=Capsule, length=3.75, radius=6.75, shapeName=neck_01_capsule, center=<0 3.1471 0.7193>, rotation=<rot 90 180 -149.9992>}]
 spine_05 [
  {shapeType=Capsule, length=8, radius=9, shapeName=spine_05_capsule, center=<-2.121 -0.3034 0>}
  {shapeType=Capsule, length=7.25, radius=8.25, shapeName=spine_05_capsule, center=<-6.4718 4.2657 0>}
 ]
 upperarm_l [{
   shapeType=Capsule
   length=4.25
   radius=5.5
   shapeName=upperarm_l_capsule
   center=<-2.0667 0.5174 -3.3634>
   rotation=<rot 60 -179.9996 -179.9997>
  }]
}
constraints [
 {bone1Name=dyn_scarf_bk_01, bone2Name=dyn_scarf_knot, swing1LimitDegrees=10, swing2LimitDegrees=10, twistLimitDegrees=10}
 {bone1Name=dyn_scarf_bk_02, bone2Name=dyn_scarf_bk_01, swing1LimitDegrees=20, swing2LimitDegrees=20, twistLimitDegrees=10}
 {bone1Name=dyn_scarf_knot, bone2Name=neck_01, swing1LimitDegrees=10, swing2LimitDegrees=10, twistLimitDegrees=5}
 {
  bone1Name=upperarm_l
  bone2Name=spine_05
  swing1LimitDegrees=45
  swing1Motion=Locked
  swing2LimitDegrees=45
  swing2Motion=Locked
  twistLimitDegrees=45
  twistMotion=Locked
 }
 {
  bone1Name=neck_01
  bone2Name=spine_05
  swing1LimitDegrees=45
  swing1Motion=Locked
  swing2LimitDegrees=45
  swing2Motion=Locked
  twistLimitDegrees=45
  twistMotion=Locked
 }
 {bone1Name=dyn_scarf_knot, bone2Name=dyn_scarf_fr_01, swing1LimitDegrees=10, swing2LimitDegrees=10, twistLimitDegrees=5}
 {bone1Name=dyn_scarf_fr_01, bone2Name=dyn_scarf_fr_02, swing1LimitDegrees=10, swing2LimitDegrees=30, twistLimitDegrees=5}
 {bone1Name=dyn_scarf_fr_02, bone2Name=dyn_scarf_fr_03, swing1LimitDegrees=10, swing2LimitDegrees=30, twistLimitDegrees=5}
]
