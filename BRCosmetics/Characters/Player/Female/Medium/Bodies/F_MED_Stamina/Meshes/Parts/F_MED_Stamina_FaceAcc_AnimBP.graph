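# Animation Blueprint for the Stamina female medium body's face accessory part, adding secondary motion to hair bones.
# It copies the pose from the attached parent mesh, then chains four Anim Dynamics nodes on front, front-right, back-right and back hair bones.
# Each simulates a single body with tight angular limits before reaching the output pose.
# An additive head-bone offset node is left disconnected, and the event graph has no logic.
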AnimBlueprint /BRCosmetics/Characters/Player/Female/Medium/Bodies/F_MED_Stamina/Meshes/Parts/F_MED_Stamina_FaceAcc_AnimBP
target_skeleton=&/F_MED_Stamina_FaceAcc_Skeleton
graphs [
 AnimGraph {nodes [
   n1 = "Transform (Modify) Bone - Bone: head"(ComponentPose=Local To Component(), Alpha=1, Translation="0,1,-3", Rotation="0,0,0", Scale="1,1,1") {translationMode=BMM_Additive, alphaBoolBlend {blendOption=Linear}, boneToModify {boneName=head}, translation {y=1, z=-3}}
   Component To Local(ComponentPose=@n1)
   n2 = Anim Dynamics - Bone: dyn_hair_front_01(ComponentPose=Local To Component(LocalPose=Copy Pose From Mesh() {bUseAttachedParent=true}), Alpha=1, ExternalForce="0,0,0") {
    alphaBoolBlend {blendOption=Linear}
    boundBone {boneName=dyn_hair_front_01}
    physicsBodyDefinitions [{
      boundBone {boneName=dyn_hair_front_01}
      boxExtents=<48 6 48>
      constraintSetup {angularLimitsMax=<0 6 0>, angularLimitsMin=<0 -6 -10>, angularTarget=<1 0 0>}
      localJointOffset=<5 -0.5 -2>
     }]
   }
   n3 = Anim Dynamics - Bone: dyn_hair_front_r_01(ComponentPose=@n2, Alpha=1, ExternalForce="0,0,0") {
    alphaBoolBlend {blendOption=Linear}
    boundBone {boneName=dyn_hair_front_r_01}
    physicsBodyDefinitions [{
      boundBone {boneName=dyn_hair_front_r_01}
      boxExtents=<48 6 48>
      constraintSetup {angularLimitsMax=<0 2 2>, angularLimitsMin=<0 -2 -10>, angularTarget=<1 0 0>}
      localJointOffset=<8 -0.5 -8>
     }]
   }
   n4 = Anim Dynamics - Bone: dyn_hair_bck_r_01(ComponentPose=@n3, Alpha=1, ExternalForce="0,0,0") {
    alphaBoolBlend {blendOption=Linear}
    boundBone {boneName=dyn_hair_bck_r_01}
    physicsBodyDefinitions [{
      boundBone {boneName=dyn_hair_bck_r_01}
      boxExtents=<48 6 48>
      constraintSetup {angularLimitsMax=<0 2 2>, angularLimitsMin=<0 -2 -10>, angularTarget=<1 0 0>}
      localJointOffset=<8 -0.5 -8>
     }]
   }
   n5 = Anim Dynamics - Bone: dyn_hair_bck_01(ComponentPose=@n4, Alpha=1, ExternalForce="0,0,0") {
    alphaBoolBlend {blendOption=Linear}
    boundBone {boneName=dyn_hair_bck_01}
    physicsBodyDefinitions [{
      boundBone {boneName=dyn_hair_bck_01}
      boxExtents=<48 6 48>
      constraintSetup {angularLimitsMax=<0 2 2>, angularLimitsMin=<0 -2 -10>, angularTarget=<1 0 0>}
      localJointOffset=<8 -0.5 -8>
     }]
   }
   Output Pose(Result=Component To Local(ComponentPose=@n5)) {layerGroup=None}
  ]}
 EventGraph {nodes [Event BlueprintUpdateAnimation()]}
]
variables []
